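# Niagara effect for the UrchinQuota guitar cosmetic, adding cyan flames, pulsing purple eye glows and layered head flames and glows.
# BodyFlamesLocal spawns flames continuously at seven fixed points via partition-indexed arrays, while eye and head emitters burst persistent particles at hardcoded offsets.
# Flames pass a smoothed, lagging position into Interpolate Dynamic, and brightness varies by time of day, front-end context, platform and distance.
# Lightning, hand and older HeadOffset-driven head emitters are disabled.
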NiagaraSystem /SparksCosmetics/Instrument/Guitar/UrchinQuota/Effects/NS_Guitar_UrchinQuota
dependencies {
 usedDataInterfaces [
  &NiagaraDataInterfaceCurve
  &NiagaraDataInterfaceArrayFloat
  &NiagaraDataInterfaceMeshRendererInfo
  &NiagaraDataInterfacePlatformSet
  &NiagaraDataInterfaceArrayFloat3
  &NiagaraDataInterfaceArrayFloat2
  &NiagaraDataInterfaceVectorCurve
  &NiagaraDataInterfaceArrayColor
 ]
 usedDynamicInputs [&SimulationPosition, &Add_Integer, &Multiply_Int]
 usedModules [
  &SystemState
  &self:SystemSpawnScript.SystemScriptSource.SystemScriptGraph.NiagaraNodeAssignment_0.SetVariables_#1
 ]
}
emitters {
 BodyFlamesLocal {
  emitterUpdateScript [
   &EmitterState()
   &SpawnBurst_Instantaneous(Spawn Count=6, Use Spawn Probability=false, UseLoopCountLimit=true)
   &SpawnRate(SpawnRate=20, Use Spawn Probability=false) {enabled=false, module=SpawnRate001}
   &SpawnRate(Spawn Probability=0.8, SpawnRate=14, Use Spawn Probability=true)
  ]
  particleSpawnScript [
   &PartitionParticles(Number Of Partitions=7)
   /SetVariables(
    Particles.Drama:NiagaraFloat=&RandomRangeFloat(Minimum=0, Maximum=1)
    Particles.FirePoint:Vector3f=&SelectVectorFromArray(
     Direct Array Index=$Particles.PartitionParticles.Partition
     Vector Selection Array=DataInterface() {propertyValues {FloatData [<-5 -2 5>, <-8 -2 13>, <3 -2 9>, <9 -2 8>, <3 -2 -19>, <-8 -2 -15>, <-8 -2 -10>]}}
    )
   ) {isSetParameters=true}
   &InitializeParticle(
    Color=&MakeCustomLinearColorFromBool(Bool=$User.bIsFrontEndPreview, True Color=<rgba 0.6092 1.1791 1.5 40>, False Color=<rgba 0.6092 1.1791 1.5 40>)
    Lifetime=&Lerp_Float(A=0.5, B=3, Alpha=$Particles.Drama)
    Mesh Scale=&Lerp_Vector(A=<0.2 1 1>, B=<0.55 1 1>, Alpha=$Particles.Drama)
    Mesh Scale Mode=::NewEnumerator4
    Position Offset=$Particles.FirePoint
    Sprite Size Mode=::NewEnumerator3
    Uniform Sprite Size=3
    UsePositionOffset=true
    Write Material Random=true
   )
   &InitialMeshOrientation(
    Mesh Orientation Mode=::NewEnumerator2
    Orientation Coordinate Space=::World
    Orientation Vector=&RotateVectorByQuaternion(
     Vector To Rotate=<0 0 1>
     Quaternion=&LerpQuaternion(
      Quaternion A:Quat4f=<0 0 0 1>
      Quaternion B:Quat4f=$Engine.Owner.Rotation
      Lerp Factor:NiagaraFloat=0.5
      Enable Spherical Linear Interpolation:NiagaraBool=false
     )
    )
    Use Rotation Vector=false
   ) {enabled=false}
   &InitialMeshOrientation(
    Mesh Orientation Mode=::NewEnumerator2
    Orientation Coordinate Space=::World
    Orientation Vector=<0 0 1>
    Rotation=&RandomRangeVector(Minimum=<0 -0.1 0>, Maximum=<0 0 0>)
    Use Rotation Vector=false
   ) {enabled=false, module=InitialMeshOrientation001}
   &CameraOffset(Camera Offset Amount=-3)
  ]
  particleUpdateScript [
   &ParticleState()
   &ScaleMeshSize(Scale Factor=&SelectVectorFromArray(
     Direct Array Index=$Particles.PartitionParticles.Partition
     Vector Selection Array=DataInterface() {propertyValues {FloatData [<0.65 0.35 0.35>, <0.8 0.1 0.1>, <0.7 0.2 0.2>, <0.45 0.1 0.2>, <0.65 0.2 0.2>, <0.8 0.3 0.3>, <0.65 0.4 0.4>]}}
    )) {module=ScaleMeshSize001}
   &UpdateMeshOrientation(
    Facing Direction=&RotateVectorByQuaternion(
     Vector To Rotate=<0 0 1>
     Quaternion=&LerpQuaternion(
      Quaternion A:Quat4f=<0 -1 0 1>
      Quaternion B:Quat4f=<0 1 0 1>
      Lerp Factor:NiagaraFloat=&SelectFloatFromArray(
       Direct Array Index=$Particles.PartitionParticles.Partition
       Float Selection Array=DataInterface() {propertyValues {FloatData [-0.975,-0.3,-0.375,-1,0.2,0.15,0.25]}}
      )
      Enable Spherical Linear Interpolation:NiagaraBool=false
     )
    )
    Facing Interpolation=false
    Orientation Method=::NewEnumerator1
    Override Quat=false
    TransformFacing=false
    TransformUp=false
   )
   &ScaleMeshSize(Scale Factor=&Lerp_Vector(
     A=&VectorFromCurve(VectorCurve=DataInterface() {propertyValues {
        XCurve {keys [
          {arriveTangent=4.3316, interpMode=RCIM_Cubic, leaveTangent=4.3316, tangentMode=RCTM_User, time=0, value=0}
          {arriveTangent=0.5, interpMode=RCIM_Cubic, leaveTangent=0.5, time=0.2, value=1}
          <at 1 0.5 RCIM_Cubic>
         ]}
        YCurve {keys [<at 0 1>]}
        ZCurve {keys [<at 0 1>]}
       }})
     B=&VectorFromCurve(
      VectorCurve=DataInterface() {propertyValues {
        XCurve {keys [
          {arriveTangent=4.3316, interpMode=RCIM_Cubic, leaveTangent=4.3316, tangentMode=RCTM_User, time=0, value=0}
          {arriveTangent=0.5, interpMode=RCIM_Cubic, leaveTangent=0.5, time=0.5, value=1}
          {arriveTangent=-1.2021, interpMode=RCIM_Cubic, leaveTangent=-1.2021, tangentMode=RCTM_User, time=1, value=0.5}
         ]}
        YCurve {keys [<at 0 1>]}
        ZCurve {keys [<at 0 1>]}
       }}
      Scale Curve=&VectorFromFloat(Value=$Particles.Drama)
     )
     Alpha=$Particles.Drama
    ))
   /SetVariables(Particles.SoftColor:LinearColor=&MultiplyLinearColor(
     Linear Color=&MakeCustomLinearColorFromBool(Bool=$User.bIsFrontEndPreview, True Color=<rgba 0.3332 1.9801 2 10>, False Color=<rgba 0.3332 1.9801 2 10>)
     Scale Factor=&MakeFloat4FromVectorAndFloat(
      "Vector (RGB)"=<1 1 1>
      "Float (Alpha)"=&FloatFromCurve(
       FloatCurve=DataInterface() {propertyValues {Curve {keys [
           {arriveTangent=2.2421, interpMode=RCIM_Cubic, leaveTangent=2.2421, tangentMode=RCTM_User, time=0, value=0.5}
           {arriveTangent=-0.5, interpMode=RCIM_Cubic, leaveTangent=-0.5, time=0.2, value=1}
           <at 1 0 RCIM_Cubic>
          ]}}}
       Scale Curve=&MakeFloatFromVector4(
        Vector 4=&MakeVector4_FromFloat(Float=&FNTimeOfDayScalar(Night Scale=0.7, Morning Scale=1, Evening Scale=1))
        Channel=::NewEnumerator0
       )
      )
     )
    )) {isSetParameters=true}
   &DynamicMaterialParameters(
    Index 0 Param 2=&RandomRangeFloat(Minimum=-0.2, Maximum=0.2)
    Index 0 Param 3=1
    Index 2 Param 1=&RandomRangeFloat(Minimum=0.5, Maximum=1.3)
    Index 2 Param 2=&RandomRangeFloat(Minimum=0.1, Maximum=0.3)
    Index 2 Param 3=1
    Index 2 Param 4=&RandomRangeFloat(Minimum=0, Maximum=-0.1)
    Module.Index0Param0Write=true
    Module.Index0Param1Write=true
    Module.Index0Param2Write=true
    Module.Index0Param3Write=true
    Module.Index2Param0Write=true
    Module.Index2Param1Write=true
    Module.Index2Param2Write=true
    Module.Index2Param3Write=true
    Write Parameter Index 2=true
   )
   &InterpolateOverTime(Rate Of Change=5, Target Vector Value:NiagaraPosition=$Particles.Position, Variable Type=::NewEnumerator2)
   /SetVariables(Particles.Interpolate Dynamic=&MakeFloat4FromVectorAndFloat("Vector (RGB)"=$Particles.InterpolateOverTime.InterpolateOverTime_V3_.Moving Average, "Float (Alpha)"=1.2)) {isSetParameters=true}
   &ScaleColor(
    Scale Alpha=&FloatFromCurve(FloatCurve=DataInterface() {propertyValues {Curve {keys [
         {arriveTangent=2.2421, interpMode=RCIM_Cubic, leaveTangent=2.2421, tangentMode=RCTM_User, time=0, value=0.5}
         {arriveTangent=-0.5, interpMode=RCIM_Cubic, leaveTangent=-0.5, time=0.2, value=1}
         <at 1 0 RCIM_Cubic>
        ]}}})
    ScaleA=true
    ScaleRGB=false
   )
   /SetVariables(Particles.Position=&AddVectorToPosition(
     Position=&AddVectorToPosition(
      Position=/Niagara/DynamicInputs/Helpers/SimulationPosition()
      Vector=&RotateVectorByQuaternion(Vector To Rotate=$Particles.FirePoint, Quaternion=$Engine.Owner.Rotation)
     )
     Vector=&Lerp_Vector(
      A=<0 0 0>
      B=&RotateVectorByQuaternion(Vector To Rotate=&Lerp_Vector(A=<2 0 0>, B=<1 0 0>, Alpha=0.5), Quaternion=$Particles.MeshOrientation)
      Alpha=$Particles.NormalizedAge
     )
    )) {enabled=false, isSetParameters=true}
   &CameraOffset(Camera Offset Amount=-1) {enabled=false, module=CameraOffset001}
  ]
  rendererClasses [&NiagaraSpriteRendererProperties, &NiagaraMeshRendererProperties]
  renderers [
   {rendererClass=&NiagaraSpriteRendererProperties}
   {rendererClass=&NiagaraMeshRendererProperties, rendererIndex=1}
   {rendererClass=&NiagaraMeshRendererProperties, rendererIndex=2}
  ]
 }
 EyesGlow {
  emitterSpawnScript [/SetVariables(Emitter.EyesOffset:Vector3f=<8 0.5 -5.5>, Emitter.LeftOffset:Vector3f=<1.8 0 -0.6>, Emitter.RightOffset:Vector3f=<-0.7 0 -2.1>) {isSetParameters=true}]
  emitterUpdateScript [
   &EmitterState()
   &SpawnBurst_Instantaneous(Spawn Count=4, Use Spawn Probability=false, UseLoopCountLimit=true)
  ]
  particleSpawnScript [
   &PartitionParticles(Number Of Partitions=4, Particles Per Partition=1, Partition Mode:ENiagara_PartitionMode=::NewEnumerator1)
   &InitializeParticle(
    Color=<rgba 3.1976 0.5729 5 1>
    Lifetime=5
    Position Offset=&Add_Vector(
     A=&Add_Vector(
      A=$Emitter.EyesOffset
      B=&MakeCustomVectorFromBool(
       Bool=&Boolean_OrOperation(
        Bool1=&SetBoolByIntComparison(A=$Particles.PartitionParticles.Partition, B=0, Comparison Type=::NewEnumerator4)
        Bool2=&SetBoolByIntComparison(A=$Particles.PartitionParticles.Partition, B=2, Comparison Type=::NewEnumerator4)
       )
       True Vector=$Emitter.LeftOffset
       False Vector=$Emitter.RightOffset
      )
     )
     B=&SelectVectorFromArray(
      Direct Array Index=$Particles.PartitionParticles.Partition
      Vector Selection Array=DataInterface() {propertyValues {FloatData [<0 0.2 7>, <0 0.2 -7>, <0 -0.5 7>, <0 -0.5 -7>]}}
     )
    )
    Sprite Size=&SelectVector2DFromArray(
     Direct Array Index=$Particles.PartitionParticles.Partition
     Vector 2D Selection Array=DataInterface() {propertyValues {FloatData [<8 8>, <8 8>, <9 16>, <9 16>]}}
    )
    Sprite Size Mode=::NewEnumerator4
    UsePositionOffset=true
    Write Material Random=true
   )
  ]
  particleUpdateScript [
   &ParticleState(Kill Particles When Lifetime Has Elapsed=false, Let Infinitely Lived Particles Die When Emitter Deactivates=true)
   /SetVariables(
    Particles.SpriteAlignment=&SelectVectorFromArray(
     Direct Array Index=$Particles.PartitionParticles.Partition
     Vector Selection Array=DataInterface() {propertyValues {FloatData [<50 0 0>, <50 0 0>, <25 -30 -50>, <50 -30 50>]}}
    )
    Particles.SpriteFacing=<0.1293 1 0>
   ) {isSetParameters=true}
   &DynamicMaterialParameters(
    Index 0 Param 1=&FN_FrontEndFloatModifier(Gameplay=3, Front End=4)
    Index 0 Param 2=3
    Module.Index0Param0Write=true
    Module.Index0Param1Write=true
    Module.Index0Param2Write=true
    Module.Index0Param3Write=true
   )
   &ScaleColor(
    Scale Alpha=&Multiply_Float(
     A=&FNTimeOfDayScalar(Night Scale=0.8, Morning Scale=1, Evening Scale=1)
     B=&FN_FrontEndFloatModifier(Gameplay=1, Front End=0.5)
    )
    Scale RGB=&VectorFromFloat(Value=&Waveform(
      Waveform Count=::NewEnumerator1
      "Waveform [1]"=::NewEnumerator0
      "[1] Frequency"=0.3333
      "[1] Amplitude Min/Max"=<0.5 3>
      "Waveform [2] Blend"=::NewEnumerator4
      "[2] Input Position"=0
      "[2] Phase"=$Emitter.Age
      "[2] Phase Scale"=1
      "[2] Compound Frequency"=2
      "[2] Frequency"=0.2
      "[2] Amplitude Min/Max"=<1 3>
      "[2] Scale"=1
      "[2] Bias"=0
      "[2] Exponent"=1
      "[2] Pulse Duration"=0.1
      "[2] Clip Triangle Min/Max"=<0 0>
      "[2] Random Seed"=0
     ))
    ScaleA=true
    ScaleRGB=true
   )
  ]
  rendererClasses [&NiagaraSpriteRendererProperties]
  renderers [{rendererClass=&NiagaraSpriteRendererProperties}]
 }
 HeadMain {
  bEnabled=false
  emitterUpdateScript [
   &EmitterState()
   &SpawnBurst_Instantaneous(Use Spawn Probability=false, UseLoopCountLimit=true)
   /SetVariables(Emitter.WeakPointReactivity=-0.5) {isSetParameters=true, module=SetVariables_#2}
   /SetVariables(Emitter.DissolveTextureSelect=1) {isSetParameters=true, module=SetVariables_#3}
  ]
  particleSpawnScript [
   &PartitionParticles()
   &InitializeParticle(
    Color=<rgba 1.2 1.2 1.2 20>
    Lifetime=5
    Mesh Scale=<0.6 1 1>
    Mesh Scale Mode=::NewEnumerator4
    Position Offset=$System.HeadOffset
    UsePositionOffset=true
    Write Material Random=true
   )
   &InitialMeshOrientation(Mesh Orientation Mode=::NewEnumerator2, Orientation Vector=<0 0 1>, Use Rotation Vector=false)
   &DynamicMaterialParameters(
    Index 0 Param 2=0.12
    Index 0 Param 3=1
    Module.Index0Param0Write=true
    Module.Index0Param1Write=true
    Module.Index0Param2Write=true
    Module.Index0Param3Write=true
   )
   &CameraOffset(Camera Offset Amount=&FN_PickFloatDependOnPlatform(TargetPlatformActive=0, TargetPlatformNoActive=-5))
  ]
  particleUpdateScript [
   &ParticleState(Kill Particles When Lifetime Has Elapsed=false, Let Infinitely Lived Particles Die When Emitter Deactivates=true)
   /SetVariables(Particles.Color003:LinearColor=&MultiplyLinearColor(
     Linear Color=<rgba 1.1875 2 2 15>
     Scale Factor=&MakeVector4_FromFloat(Float=&Multiply_Float(
       A=&FNTimeOfDayScalar(Night Scale=1, Morning Scale=1, Evening Scale=1)
       B=&FN_FrontEndFloatModifier(
        Gameplay=1
        Front End=&FN_LOD_Distance_Remap(Start Distance=150, End Distance=200, Minimum Output=0.5, Maximum Output=1, Clamp Output=true)
       )
      ))
    )) {isSetParameters=true, module=SetVariables_#4}
   &ScaleColor(
    Scale Alpha=&Multiply_Float(
     A=&FNTimeOfDayScalar(Night Scale=0.6, Morning Scale=1, Evening Scale=1)
     B=&FN_FrontEndFloatModifier(
      Gameplay=&Multiply_Float(
       A=&FN_PickFloatDependOnPlatform(TargetPlatformActive=1, TargetPlatformNoActive=0.5)
       B=&FN_LOD_Distance_Remap(Start Distance=150, End Distance=200, Minimum Output=0.3, Maximum Output=1, Clamp Output=true)
      )
      Front End=&FN_LOD_Distance_Remap(Start Distance=150, End Distance=200, Minimum Output=0.5, Maximum Output=2, Clamp Output=true)
     )
    )
    ScaleA=true
    ScaleRGB=true
   )
   &InterpolateOverTime(Rate Of Change=10, Target Vector Value:NiagaraPosition=$Particles.Position, Variable Type=::NewEnumerator2)
   /SetVariables(Particles.Interpolate Dynamic=&MakeFloat4FromVectorAndFloat("Vector (RGB)"=$Particles.InterpolateOverTime.InterpolateOverTime_V3_.Moving Average, "Float (Alpha)"=1.2)) {isSetParameters=true, module=SetVariables_#5}
   /SetVariables(Particles.Position=&AddVectorToPosition(Position=&SimulationPosition(), Vector=$System.HeadOffset)) {isSetParameters=true, module=SetVariables_#6}
  ]
  rendererClasses [&NiagaraMeshRendererProperties]
  renderers [
   {rendererClass=&NiagaraMeshRendererProperties, rendererIndex=0}
   {rendererClass=&NiagaraMeshRendererProperties, rendererIndex=1}
  ]
 }
 HeadMain001 {
  bEnabled=false
  emitterUpdateScript [
   &EmitterState()
   &SpawnBurst_Instantaneous(Use Spawn Probability=false, UseLoopCountLimit=true)
   /SetVariables(Emitter.WeakPointReactivity=-0.5) {isSetParameters=true, module=SetVariables_#2}
   /SetVariables(Emitter.DissolveTextureSelect=1) {isSetParameters=true, module=SetVariables_#3}
  ]
  particleSpawnScript [
   &PartitionParticles()
   &InitializeParticle(
    Color=<rgba 1.2 1.2 1.2 20>
    Lifetime=5
    Mesh Scale=<0.25 0.2 0.2>
    Mesh Scale Mode=::NewEnumerator4
    Position Offset=$System.HeadOffset
    UsePositionOffset=false
    Write Material Random=true
   )
   &InitialMeshOrientation(Mesh Orientation Mode=::NewEnumerator2, Orientation Vector=<0 0 1>, Use Rotation Vector=false)
   &DynamicMaterialParameters(
    Index 0 Param 2=0.12
    Index 0 Param 3=1
    Module.Index0Param0Write=true
    Module.Index0Param1Write=true
    Module.Index0Param2Write=true
    Module.Index0Param3Write=true
   )
   &CameraOffset(Camera Offset Amount=&FN_PickFloatDependOnPlatform(TargetPlatformActive=0, TargetPlatformNoActive=-5))
  ]
  particleUpdateScript [
   &ParticleState(Kill Particles When Lifetime Has Elapsed=false, Let Infinitely Lived Particles Die When Emitter Deactivates=true)
   /SetVariables(Particles.Color003:LinearColor=&MultiplyLinearColor(
     Linear Color=<rgba 1.1875 2 2 15>
     Scale Factor=&MakeVector4_FromFloat(Float=&Multiply_Float(
       A=&FNTimeOfDayScalar(Night Scale=0.5, Morning Scale=0.5, Day Scale=0.5, Evening Scale=0.5)
       B=&FN_FrontEndFloatModifier(
        Gameplay=1
        Front End=&FN_LOD_Distance_Remap(Start Distance=150, End Distance=200, Minimum Output=0.5, Maximum Output=1, Clamp Output=true)
       )
      ))
    )) {isSetParameters=true, module=SetVariables_#4}
   &ScaleColor(
    Scale Alpha=&Multiply_Float(
     A=&FNTimeOfDayScalar(Night Scale=0.6, Morning Scale=1, Evening Scale=1)
     B=&FN_FrontEndFloatModifier(
      Gameplay=&Multiply_Float(
       A=&FN_PickFloatDependOnPlatform(TargetPlatformActive=1, TargetPlatformNoActive=0.5)
       B=&FN_LOD_Distance_Remap(Start Distance=150, End Distance=200, Minimum Output=0.3, Maximum Output=1, Clamp Output=true)
      )
      Front End=&FN_LOD_Distance_Remap(Start Distance=150, End Distance=200, Minimum Output=0.5, Maximum Output=2, Clamp Output=true)
     )
    )
    ScaleA=true
    ScaleRGB=true
   )
   &InterpolateOverTime(Rate Of Change=10, Target Vector Value:NiagaraPosition=$Particles.Position, Variable Type=::NewEnumerator2)
   /SetVariables(Particles.Interpolate Dynamic=&MakeFloat4FromVectorAndFloat("Vector (RGB)"=$Particles.InterpolateOverTime.InterpolateOverTime_V3_.Moving Average, "Float (Alpha)"=1.2)) {isSetParameters=true, module=SetVariables_#5}
   /SetVariables(Particles.Position=&AddVectorToPosition(Position=&SimulationPosition(), Vector=$System.HeadOffset)) {isSetParameters=true, module=SetVariables_#6}
  ]
  rendererClasses [&NiagaraMeshRendererProperties]
  renderers [
   {rendererClass=&NiagaraMeshRendererProperties, rendererIndex=0}
   {rendererClass=&NiagaraMeshRendererProperties, rendererIndex=1}
  ]
 }
 Head_Main_Local {
  emitterUpdateScript [
   &EmitterState()
   &SpawnBurst_Instantaneous(Use Spawn Probability=false, UseLoopCountLimit=true)
   /SetVariables(Emitter.WeakPointReactivity=-0.5) {isSetParameters=true, module=SetVariables_#2}
   /SetVariables(Emitter.DissolveTextureSelect=1) {isSetParameters=true, module=SetVariables_#3}
  ]
  particleSpawnScript [
   &PartitionParticles()
   &InitializeParticle(
    Color=<rgba 1.2 1.2 1.2 20>
    Lifetime=5
    Mesh Scale=<0.25 0.2 0.2>
    Mesh Scale Mode=::NewEnumerator4
    Position Offset=<-3 1 60>
    UsePositionOffset=true
    Write Material Random=true
   )
   &InitialMeshOrientation(Mesh Orientation Mode=::NewEnumerator2, Orientation Vector=<0 0 1>, Use Rotation Vector=false)
   &DynamicMaterialParameters(
    Index 0 Param 1=&RandomRangeFloat(Minimum=0, Maximum=1)
    Index 0 Param 2=0.12
    Index 0 Param 3=0.35
    Module.Index0Param0Write=true
    Module.Index0Param1Write=true
    Module.Index0Param2Write=true
    Module.Index0Param3Write=true
   )
   &CameraOffset(Camera Offset Amount=&FN_PickFloatDependOnPlatform(TargetPlatformActive=0, TargetPlatformNoActive=-5))
  ]
  particleUpdateScript [
   &ParticleState(Kill Particles When Lifetime Has Elapsed=false, Let Infinitely Lived Particles Die When Emitter Deactivates=true)
   &UpdateMeshOrientation(
    Facing Direction=&RotateVectorByQuaternion(
     Vector To Rotate=<0 0 1>
     Quaternion=&LerpQuaternion(
      Quaternion A:Quat4f=<0 -1 0 1>
      Quaternion B:Quat4f=<0 1 0 1>
      Lerp Factor:NiagaraFloat=1
      Enable Spherical Linear Interpolation:NiagaraBool=false
     )
    )
    Facing Interpolation=false
    Orientation Method=::NewEnumerator1
    Override Quat=false
    TransformFacing=false
    TransformUp=false
   )
   &ScaleMeshSize(Scale Factor=<1 0.5 0.5>)
   /SetVariables(Particles.Color003:LinearColor=&MultiplyLinearColor(
     Linear Color=<rgba 1.1875 2 2 15>
     Scale Factor=&MakeVector4_FromFloat(Float=&Multiply_Float(
       A=&FNTimeOfDayScalar(Night Scale=0.5, Morning Scale=0.5, Day Scale=0.5, Evening Scale=0.5)
       B=&FN_FrontEndFloatModifier(
        Gameplay=1
        Front End=&FN_LOD_Distance_Remap(Start Distance=150, End Distance=200, Minimum Output=0.5, Maximum Output=1, Clamp Output=true)
       )
      ))
    )) {isSetParameters=true, module=SetVariables_#4}
   &ScaleColor(
    Scale Alpha=&Multiply_Float(
     A=&FNTimeOfDayScalar(Night Scale=0.6, Morning Scale=1, Evening Scale=1)
     B=&FN_FrontEndFloatModifier(
      Gameplay=&Multiply_Float(
       A=&FN_PickFloatDependOnPlatform(TargetPlatformActive=1, TargetPlatformNoActive=0.5)
       B=&FN_LOD_Distance_Remap(Start Distance=150, End Distance=200, Minimum Output=0.3, Maximum Output=1, Clamp Output=true)
      )
      Front End=&FN_LOD_Distance_Remap(Start Distance=150, End Distance=200, Minimum Output=0.5, Maximum Output=2, Clamp Output=true)
     )
    )
    ScaleA=true
    ScaleRGB=true
   )
   &InterpolateOverTime(Rate Of Change=10, Target Vector Value:NiagaraPosition=$Particles.Position, Variable Type=::NewEnumerator2)
   /SetVariables(Particles.Interpolate Dynamic=&MakeFloat4FromVectorAndFloat("Vector (RGB)"=$Particles.InterpolateOverTime.InterpolateOverTime_V3_.Moving Average, "Float (Alpha)"=1.2)) {isSetParameters=true, module=SetVariables_#5}
   /SetVariables(Particles.Position=&AddVectorToPosition(Position=&SimulationPosition(), Vector=$System.HeadOffset)) {enabled=false, isSetParameters=true, module=SetVariables_#6}
  ]
  rendererClasses [&NiagaraMeshRendererProperties]
  renderers [
   {rendererClass=&NiagaraMeshRendererProperties, rendererIndex=0}
   {rendererClass=&NiagaraMeshRendererProperties, rendererIndex=1}
  ]
 }
 Head_PurpleGlow_Local {
  emitterUpdateScript [
   &EmitterState()
   &SpawnBurst_Instantaneous(Use Spawn Probability=false, UseLoopCountLimit=true)
  ]
  particleSpawnScript [
   &InitializeParticle(
    Color=<rgba 3.0528 0.0204 8 1.5>
    Mesh Scale=<0.88 0.5 1>
    Mesh Scale Mode=::NewEnumerator4
    Position Offset=<-2 -2 55>
    UsePositionOffset=true
    Write Material Random=true
   )
   &CameraOffset(Camera Offset Amount=3) {enabled=false}
   &InitialMeshOrientation(Mesh Axis To Orient=<0 0 0>, Mesh Orientation Mode=::NewEnumerator2, Use Rotation Vector=false) {enabled=false}
  ]
  particleUpdateScript [
   &ParticleState(Kill Particles When Lifetime Has Elapsed=false, Let Infinitely Lived Particles Die When Emitter Deactivates=true)
   &ScaleMeshSize(Scale Factor=$Engine.Owner.Scale) {module=ScaleMeshSize001}
   &ScaleMeshSize(Scale Factor=<0.15 0.15 0.175>)
   /SetVariables(Particles.Position=&AddVectorToPosition(Position=&SimulationPosition(), Vector=&Add_Vector(A=$System.HeadOffset, B=<0 0 -5>))) {enabled=false, isSetParameters=true, module=SetVariables_#7}
  ]
  rendererClasses [&NiagaraMeshRendererProperties]
  renderers [{rendererClass=&NiagaraMeshRendererProperties, rendererIndex=0}]
 }
 Lightning {
  bEnabled=false
  emitterSpawnScript [/SetVariables(Emitter.LightningTilingEye=1, Emitter.LightningWidthEye=12) {isSetParameters=true}]
  emitterUpdateScript [
   &EmitterState()
   &SpawnRate(SpawnRate=2, Use Spawn Probability=false) {module=SpawnRate001}
  ]
  particleSpawnScript [
   &InitializeParticle(
    Lifetime Max=0.6
    Lifetime Min=0.5
    Lifetime Mode=::NewEnumerator1
    Mesh Scale Max=<4 2.5 4>
    Mesh Scale Min=<4 2.5 2>
    Mesh Scale Mode=::NewEnumerator2
    Position Offset=<27.0862 -15 -9>
    UsePositionOffset=true
    Write Material Random=true
   )
   /SetVariables(Particles.MeshIndex=&RandomRangeInt(Minimum=0, Maximum=2)) {isSetParameters=true}
   &InitialMeshOrientation(Mesh Axis To Orient=<0 0 -1>, Mesh Orientation Mode=::NewEnumerator2, Rotation=<0 -0.23 0>, Use Rotation Vector=true)
  ]
  particleUpdateScript [
   &ParticleState()
   &ScaleMeshSize(Scale Factor=&MakeCustomVectorFromBool(Bool=&RandomBool(), True Vector=<1 1 1>, False Vector=<1 1 -1>))
   &DynamicMaterialParameters(
    Index 0 Param 1=$Emitter.LightningWidthEye
    Index 0 Param 2=$Emitter.LightningTilingEye
    Index 0 Param 3=&FloatFromCurve(
     FloatCurve=DataInterface() {propertyValues {Curve {keys [<at 0.05 1>,<at 0.1 0.375>,<at 0.1255 0.618>,<at 0.15 0.375>,<at 0.45 0.05>,<at 0.498 0.2978>,<at 0.55 0.05>,<at 1 0>]}}}
     Scale Curve=&RandomRangeFloat(Minimum=0.7, Maximum=1.2)
    )
    Index 1 Param 1=1
    Index 1 Param 2=1
    Index 1 Param 3=1
    Module.Index0Param0Write=true
    Module.Index0Param1Write=true
    Module.Index0Param2Write=true
    Module.Index0Param3Write=false
    Module.Index1Param0Write=true
    Module.Index1Param1Write=true
    Module.Index1Param2Write=true
    Module.Index1Param3Write=false
    Write Parameter Index 1=true
   )
   &DynamicMaterialParameters(
    Index 0 Param 1=$Emitter.LightningWidthEye
    Index 0 Param 2=$Emitter.LightningTilingEye
    Index 0 Param 3=&FloatFromCurve(FloatCurve=DataInterface() {propertyValues {Curve {keys [<at 0 1>,<at 0.1 0.375>,<at 0.5 0.05>,<at 1 0>]}}})
    Index 1 Param 1=1
    Index 1 Param 2=1
    Index 1 Param 3=1
    Module.Index0Param0Write=true
    Module.Index0Param1Write=true
    Module.Index0Param2Write=true
    Module.Index0Param3Write=false
    Module.Index1Param0Write=true
    Module.Index1Param1Write=true
    Module.Index1Param2Write=true
    Module.Index1Param3Write=false
    Write Parameter Index 1=true
   ) {enabled=false, module=DynamicMaterialParameters001}
   &ScaleColor(
    Color Value To Scale=$StackContext.MainColor
    Scale Alpha=&FloatFromCurve(
     FloatCurve=DataInterface() {propertyValues {Curve {keys [
         {arriveTangent=9.0725, interpMode=RCIM_Cubic, leaveTangent=9.0725, tangentMode=RCTM_User, time=0, value=0}
         {arriveTangent=-2.5249, interpMode=RCIM_Cubic, leaveTangent=-0.2678, tangentMode=RCTM_Break, time=0.1, value=1}
         {arriveTangent=-1.6791, interpMode=RCIM_Cubic, leaveTangent=-0.0405, tangentMode=RCTM_Break, time=1, value=0}
        ]}}}
     Scale Curve=&Multiply_Float(
      A=&FN_FrontEndFloatModifier(Gameplay=0.6, Front End=0.5)
      B=&FN_PickFloatDependOnPlatform(TargetPlatformActive=1, TargetPlatformNoActive=0.3)
     )
    )
    Scale RGB=&MakeVectorFromLinearColorRGB(Color=<rgba 8.7286 0 20 1>)
    ScaleA=true
    ScaleRGB=true
   )
  ]
  rendererClasses [&NiagaraMeshRendererProperties]
  renderers [{rendererClass=&NiagaraMeshRendererProperties, rendererIndex=0}]
 }
 hand {
  bEnabled=false
  emitterUpdateScript [
   &EmitterState()
   &SpawnBurst_Instantaneous(Spawn Count=3, Use Spawn Probability=false, UseLoopCountLimit=true)
  ]
  particleSpawnScript [
   &PartitionParticles(Number Of Partitions=3)
   &InitializeParticle(
    Color=&MakeCustomLinearColorFromBool(Bool=$User.bIsFrontEndPreview, True Color=<rgba 0.6266 1.4899 2 8>, False Color=<rgba 0.3277 1.0879 1.5 40>)
    Lifetime=5
    Mesh Scale=&SelectVectorFromArray(
     Direct Array Index=$Particles.PartitionParticles.Partition
     Vector Selection Array=DataInterface() {propertyValues {FloatData [<0.55 1 6>, <0.55 1 -2>, <0.55 1 -2>]}}
    )
    Mesh Scale Mode=::NewEnumerator4
    Position Offset=&SelectVectorFromArray(
     Direct Array Index=$Particles.PartitionParticles.Partition
     Vector Selection Array=DataInterface() {propertyValues {FloatData [<-3.707 0 -1.7315>, <-0.7192 1 8.3552>, <-5.9647 0 -16.9306>]}}
    )
    UsePositionOffset=true
    Write Material Random=true
   )
   &InitialMeshOrientation(Mesh Orientation Mode=::NewEnumerator2, Orientation Vector=<0 0 1>, Use Rotation Vector=false)
   &CameraOffset()
   /SetVariables(Particles.VisibilityTag=$Particles.PartitionParticles.Partition) {isSetParameters=true}
  ]
  particleUpdateScript [
   &ParticleState(Kill Particles When Lifetime Has Elapsed=false, Let Infinitely Lived Particles Die When Emitter Deactivates=true)
   /SetVariables(Particles.Color003:LinearColor=&MultiplyLinearColor(
     Linear Color=&MakeCustomLinearColorFromBool(Bool=$User.bIsFrontEndPreview, True Color=<rgba 0.2422 1.1746 1.5 5>, False Color=<rgba 0.3332 1.9801 2 10>)
     Scale Factor=&MakeVector4_FromFloat(Float=&FNTimeOfDayScalar(Night Scale=0.7, Morning Scale=1, Evening Scale=1))
    )) {isSetParameters=true, module=SetVariables_#4}
   &DynamicMaterialParameters(
    Index 0 Param 2=0.12
    Index 0 Param 3=1
    Index 2 Param 1=&RandomRangeFloat(Minimum=0.5, Maximum=1.3)
    Index 2 Param 2=&RandomRangeFloat(Minimum=0.1, Maximum=0.3)
    Index 2 Param 3=1
    Index 2 Param 4=&RandomRangeFloat(Minimum=0, Maximum=-0.1)
    Module.Index0Param0Write=true
    Module.Index0Param1Write=true
    Module.Index0Param2Write=true
    Module.Index0Param3Write=true
    Module.Index2Param0Write=true
    Module.Index2Param1Write=true
    Module.Index2Param2Write=true
    Module.Index2Param3Write=true
    Write Parameter Index 2=true
   )
   &InterpolateOverTime(Rate Of Change=10, Target Vector Value:NiagaraPosition=$Particles.Position, Variable Type=::NewEnumerator2)
   /SetVariables(Particles.Interpolate Dynamic=&MakeFloat4FromVectorAndFloat("Vector (RGB)"=$Particles.InterpolateOverTime.InterpolateOverTime_V3_.Moving Average, "Float (Alpha)"=1.4)) {isSetParameters=true}
  ]
  rendererClasses [&NiagaraMeshRendererProperties]
  renderers [
   {rendererClass=&NiagaraMeshRendererProperties, rendererIndex=0}
   {rendererClass=&NiagaraMeshRendererProperties, rendererIndex=1}
   {rendererClass=&NiagaraMeshRendererProperties, rendererIndex=2}
   {rendererClass=&NiagaraMeshRendererProperties, rendererIndex=3}
   {rendererClass=&NiagaraMeshRendererProperties, rendererIndex=4}
   {rendererClass=&NiagaraMeshRendererProperties, rendererIndex=5}
  ]
 }
 headGlow {
  bEnabled=false
  emitterUpdateScript [
   &EmitterState()
   &SpawnBurst_Instantaneous(Use Spawn Probability=false, UseLoopCountLimit=true)
  ]
  particleSpawnScript [
   &InitializeParticle(
    Color=<rgba 3.0528 0.0204 8 1.5>
    Mesh Scale=<0.88 0.5 1>
    Mesh Scale Mode=::NewEnumerator4
    Position Offset=&Add_Vector(A=$System.HeadOffset, B=<0 0 -10>)
    UsePositionOffset=true
    Write Material Random=true
   )
   &CameraOffset(Camera Offset Amount=3) {enabled=false}
  ]
  particleUpdateScript [
   &ParticleState(Kill Particles When Lifetime Has Elapsed=false, Let Infinitely Lived Particles Die When Emitter Deactivates=true)
   &ScaleMeshSize(Scale Factor=&VectorFromFloat(Value=0.15))
   /SetVariables(Particles.Position=&AddVectorToPosition(Position=&SimulationPosition(), Vector=&Add_Vector(A=$System.HeadOffset, B=<0 0 -5>))) {isSetParameters=true, module=SetVariables_#7}
  ]
  rendererClasses [&NiagaraMeshRendererProperties]
  renderers [{rendererClass=&NiagaraMeshRendererProperties, rendererIndex=0}]
 }
 headGlow001 {
  emitterUpdateScript [
   &EmitterState()
   &SpawnBurst_Instantaneous(Spawn Count=2, Use Spawn Probability=false, UseLoopCountLimit=true)
  ]
  particleSpawnScript [
   /SetVariables(Particles.FirePoint:Vector3f=&SelectVectorFromArray(
     Direct Array Index=&ReturnExecIndex()
     Vector Selection Array=DataInterface() {propertyValues {FloatData [<-1 0 6>, <-6.7988 1 -10.4495>]}}
    )) {isSetParameters=true}
   &InitializeParticle(
    Color=<rgba 3.0528 0.0204 8 1.5>
    Mesh Scale=<0.25 0.5 1.7>
    Mesh Scale Mode=::NewEnumerator4
    Position Offset=$Particles.FirePoint
    Sprite Size Mode=::NewEnumerator3
    UsePositionOffset=true
    Write Material Random=true
   )
   &CameraOffset(Camera Offset Amount=3)
   /SetVariables(Particles.VisibilityTag=&ReturnExecIndex()) {isSetParameters=true}
  ]
  particleUpdateScript [
   &ParticleState(Kill Particles When Lifetime Has Elapsed=false, Let Infinitely Lived Particles Die When Emitter Deactivates=true)
   &ScaleMeshSize(Scale Factor=&VectorFromFloat(Value=0.2))
   /SetVariables(Particles.Position=&AddVectorToPosition(
     Position=&SimulationPosition()
     Vector=&RotateVectorByQuaternion(Vector To Rotate=$Particles.FirePoint, Quaternion=$Engine.Owner.Rotation)
    )) {isSetParameters=true, module=SetVariables_#7}
  ]
  rendererClasses [&NiagaraMeshRendererProperties, &NiagaraSpriteRendererProperties]
  renderers [
   {rendererClass=&NiagaraMeshRendererProperties, rendererIndex=0}
   {rendererClass=&NiagaraSpriteRendererProperties, rendererIndex=1}
   {rendererClass=&NiagaraMeshRendererProperties, rendererIndex=2}
  ]
 }
 headSlim {
  bEnabled=false
  emitterUpdateScript [
   &EmitterState()
   &SpawnBurst_Instantaneous(Use Spawn Probability=false, UseLoopCountLimit=true)
   /SetVariables(Emitter.WeakPointReactivity=-0.5) {isSetParameters=true, module=SetVariables_#2}
   /SetVariables() {isSetParameters=true, module=SetVariables_#3}
  ]
  particleSpawnScript [
   &PartitionParticles()
   &InitializeParticle(
    Color=<rgba 0.5475 1.8004 2 15>
    Lifetime=5
    Mesh Scale=<0.4 1 1>
    Mesh Scale Mode=::NewEnumerator4
    Position Offset=$System.HeadOffset
    Sprite Size Mode=::NewEnumerator3
    UsePositionOffset=true
    Write Material Random=true
   )
   &InitialMeshOrientation(Mesh Orientation Mode=::NewEnumerator2, Orientation Vector=<0 0 1>, Use Rotation Vector=false)
   /SetVariables(Particles.Dark_Color:LinearColor=<rgba 0.1269 0.0458 0.2 5>) {isSetParameters=true, module=SetVariables_#8}
   &DynamicMaterialParameters(
    Index 0 Param 2=0.12
    Index 0 Param 3=1
    Module.Index0Param0Write=true
    Module.Index0Param1Write=true
    Module.Index0Param2Write=true
    Module.Index0Param3Write=true
   )
  ]
  particleUpdateScript [
   &ParticleState(Kill Particles When Lifetime Has Elapsed=false, Let Infinitely Lived Particles Die When Emitter Deactivates=true)
   &ScaleColor(
    Scale Alpha=&Multiply_Float(
     A=&FNTimeOfDayScalar(Night Scale=0.6, Morning Scale=1, Evening Scale=1)
     B=&FN_FrontEndFloatModifier(Gameplay=1, Front End=0.5)
    )
    ScaleA=true
    ScaleRGB=true
   )
   &InterpolateOverTime(Rate Of Change=10, Target Vector Value:NiagaraPosition=$Particles.Position, Variable Type=::NewEnumerator2)
   /SetVariables(Particles.Interpolate Dynamic=&MakeFloat4FromVectorAndFloat("Vector (RGB)"=$Particles.InterpolateOverTime.InterpolateOverTime_V3_.Moving Average, "Float (Alpha)"=1.5)) {isSetParameters=true, module=SetVariables_#9}
   /SetVariables(Particles.Position=&AddVectorToPosition(Position=&SimulationPosition(), Vector=$System.HeadOffset)) {isSetParameters=true, module=SetVariables_#10}
  ]
  rendererClasses [&NiagaraMeshRendererProperties, &NiagaraSpriteRendererProperties]
  renderers [
   {rendererClass=&NiagaraMeshRendererProperties, rendererIndex=0}
   {rendererClass=&NiagaraSpriteRendererProperties, rendererIndex=1}
  ]
 }
 head_Slim_Local {
  emitterUpdateScript [
   &EmitterState()
   &SpawnBurst_Instantaneous(Spawn Count=2, Use Spawn Probability=false, UseLoopCountLimit=true)
   /SetVariables(Emitter.WeakPointReactivity=-0.5) {isSetParameters=true, module=SetVariables_#2}
   /SetVariables() {isSetParameters=true, module=SetVariables_#3}
  ]
  particleSpawnScript [
   &PartitionParticles()
   &InitializeParticle(
    Color=&SelectLinearColorFromArray(
     Array Sampling Mode=::NewEnumerator1
     Direct Array Index=$Particles.PartitionParticles.Partition
     Color Selection Array=DataInterface() {propertyValues {ColorData [<rgba 0.0429 2 1.178 15>, <rgba 2 0.9167 1.9099 5>]}}
     Random Seed=?
    )
    Lifetime=5
    Mesh Scale=<0.2 1 1>
    Mesh Scale Mode=::NewEnumerator4
    Position Offset=&SelectVectorFromArray(
     Direct Array Index=$Particles.PartitionParticles.Partition
     Vector Selection Array=DataInterface() {propertyValues {FloatData [<-1 1 68>, <0 1 55>]}}
    )
    Sprite Size Mode=::NewEnumerator3
    UsePositionOffset=true
    Write Material Random=true
   )
   &InitialMeshOrientation(Mesh Orientation Mode=::NewEnumerator2, Orientation Vector=<0 0 1>, Use Rotation Vector=false)
   /SetVariables(Particles.Dark_Color:LinearColor=<rgba 0.1269 0.0458 0.2 5>) {isSetParameters=true, module=SetVariables_#8}
   &DynamicMaterialParameters(
    Index 0 Param 1=&RandomRangeFloat(Minimum=0, Maximum=1)
    Index 0 Param 2=0.12
    Index 0 Param 3=0.35
    Module.Index0Param0Write=true
    Module.Index0Param1Write=true
    Module.Index0Param2Write=true
    Module.Index0Param3Write=true
   )
  ]
  particleUpdateScript [
   &ParticleState(Kill Particles When Lifetime Has Elapsed=false, Let Infinitely Lived Particles Die When Emitter Deactivates=true)
   &ScaleMeshSize(Scale Factor=&SelectVectorFromArray(
     Direct Array Index=$Particles.PartitionParticles.Partition
     Vector Selection Array=DataInterface() {propertyValues {FloatData [<1.3 1 1>, <2 1 1>]}}
    ))
   &ScaleColor(
    Scale Alpha=&Multiply_Float(
     A=&FNTimeOfDayScalar(Night Scale=0.6, Morning Scale=1, Evening Scale=1)
     B=&FN_FrontEndFloatModifier(Gameplay=1, Front End=0.5)
    )
    ScaleA=true
    ScaleRGB=false
   )
   &InterpolateOverTime(Rate Of Change=10, Target Vector Value:NiagaraPosition=$Particles.Position, Variable Type=::NewEnumerator2)
   /SetVariables(Particles.Interpolate Dynamic=&MakeFloat4FromVectorAndFloat("Vector (RGB)"=$Particles.InterpolateOverTime.InterpolateOverTime_V3_.Moving Average, "Float (Alpha)"=1.5)) {isSetParameters=true, module=SetVariables_#9}
   /SetVariables(Particles.Position=&AddVectorToPosition(Position=&SimulationPosition(), Vector=$System.HeadOffset)) {enabled=false, isSetParameters=true, module=SetVariables_#10}
  ]
  rendererClasses [&NiagaraMeshRendererProperties, &NiagaraSpriteRendererProperties]
  renderers [
   {rendererClass=&NiagaraMeshRendererProperties, rendererIndex=0}
   {rendererClass=&NiagaraSpriteRendererProperties, rendererIndex=1}
   {rendererClass=&NiagaraMeshRendererProperties, rendererIndex=2}
  ]
 }
}
summary {
 systemName=NS_Guitar_UrchinQuota
 emitters [
  {bEnabled=false, emitterName=Lightning, rendererClasses [&NiagaraMeshRendererProperties]}
  {emitterName=EyesGlow, rendererClasses [&NiagaraSpriteRendererProperties]}
  {bEnabled=false, emitterName=HeadMain, rendererClasses [&NiagaraMeshRendererProperties]}
  {
   bEnabled=false
   emitterName=headSlim
   rendererClasses [&NiagaraMeshRendererProperties, &NiagaraSpriteRendererProperties]
  }
  {bEnabled=false, emitterName=headGlow, rendererClasses [&NiagaraMeshRendererProperties]}
  {bEnabled=false, emitterName=hand, rendererClasses [&NiagaraMeshRendererProperties]}
  {emitterName=headGlow001, rendererClasses [&NiagaraMeshRendererProperties, &NiagaraSpriteRendererProperties]}
  {
   emitterName=BodyFlamesLocal
   rendererClasses [&NiagaraSpriteRendererProperties, &NiagaraMeshRendererProperties]
  }
  {bEnabled=false, emitterName=HeadMain001, rendererClasses [&NiagaraMeshRendererProperties]}
  {emitterName=Head_Main_Local, rendererClasses [&NiagaraMeshRendererProperties]}
  {
   emitterName=head_Slim_Local
   rendererClasses [&NiagaraMeshRendererProperties, &NiagaraSpriteRendererProperties]
  }
  {emitterName=Head_PurpleGlow_Local, rendererClasses [&NiagaraMeshRendererProperties]}
 ]
 userVariables [
  User.bIsFrontEnd {type=NiagaraBool}
  User.bIsFrontEndPreview {type=NiagaraBool}
  User.HandFXSize {type=NiagaraBool}
 ]
}
